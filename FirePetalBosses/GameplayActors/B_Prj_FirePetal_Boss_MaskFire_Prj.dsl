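; Homing fire projectile Blueprint for the Fire Petal boss, apparently fired by its mask attack.
; On BeginPlay it runs parent logic, then sets the trail Niagara system's "Color Variant" parameter.
; It overrides the parent's homing hooks: periodic checks call FindAndSetHomingTarget, the target is stored in ProjHomingTarget, and max speed comes from a ScalableFloat row.
; On destruction it clears both homing target references.
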(event EventBeginPlay
  (|CallParentFunction)
  (Effects|Components|ParticleSystem|SetIntParameter (Variables|Default|GetNiagara_trail_SwitchToGCNL) "Color Variant" 0))

(event EventDestroyed
  (Variables|Default|SetProjHomingTarget 0)
  (Variables|Homing|SetHomingTarget 0))

(fn GetHomingTarget
  (return (Variables|Default|GetProjHomingTarget)))

(fn GetMaxSpeed
  (return (ScalableFloat|GetValueatLevel (Variables|Default|GetProjectileMaxSpeedRow))))

(fn PeriodicCheckForHomingTarget
  (Default|FindAndSetHomingTarget))

(fn UpdateHomingTarget
  (params (HomingTarget :type Actor%20Object%20Reference))
  (Variables|Default|SetProjHomingTarget HomingTarget))

(fn ConstructionScript
  (|CallParentFunction))
